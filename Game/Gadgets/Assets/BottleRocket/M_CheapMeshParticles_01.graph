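# M_CheapMeshParticles_01 is a Default Lit material flagged for mesh particles, stored with the BottleRocket gadget assets.
# It exposes three texture parameters: Base Color, Normal Map and Specular Masks.
# Their defaults are the M80 gadget's textures rather than bottle rocket ones.
# The graph holds only these three samplers, with no math nodes and no connections shown.
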Material /Game/Gadgets/Assets/BottleRocket/M_CheapMeshParticles_01
graph {
 properties {bUsedWithMeshParticles=true, pixelDepthOffsetMode=PDOM_Legacy, usedShadingModels=Default Lit}
 nodes [
  TextureSampleParameter2D() {parameterName=Base Color, texture=/Game/Gadgets/Assets/M80/Textures/T_M80_D}
  TextureSampleParameter2D() {parameterName=Normal Map, samplerType=SAMPLERTYPE_Normal, texture=/Game/Gadgets/Assets/M80/Textures/T_M80_N}
  TextureSampleParameter2D() {parameterName=Specular Masks, samplerType=SAMPLERTYPE_Masks, texture=/Game/Gadgets/Assets/M80/Textures/T_M80_S}
 ]
}
