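# Unlit translucent particle material for small debris in the BottleRocket gadget assets, usable on Niagara and Cascade sprites.
# The small debris texture's RGB is multiplied by particle color and passed through the MF_FX_EmissiveAdjustments material function.
# A separate multiply combines texture alpha with particle alpha, so each particle can tint and fade its sprite.
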Material /Game/Gadgets/Assets/BottleRocket/M_DebrisSmall_01
graph {
 properties {
  bUsedWithNiagaraSprites=true
  bUsedWithParticleSprites=true
  blendMode=BLEND_Translucent
  pixelDepthOffsetMode=PDOM_Legacy
  shadingModel=MSM_Unlit
  usedShadingModels=Unlit
 }
 nodes [
  n1 = TextureSample() {texture=&/T_DebrisSmall_01}
  n2 = ParticleColor()
  MaterialFunctionCall(Emissive=Multiply(A=@n1.RGB, B=@n2.RGB)) {materialFunction=&MF_FX_EmissiveAdjustments}
  Multiply(A=@n1.A, B=@n2.A)
 ]
}
