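# Unlit translucent sprite material for small debris particles in the BottleRocket gadget effects, built for GPU sprites.
# MF_RandomSubUV_GPU_Sprites selects a random cell from the T_DebrisSmall_01 sheet using the Particle Random Value input and a square grid set by the Rows_Columns parameter.
# Texture color is multiplied by particle color and passed through MF_FX_EmissiveAdjustments, while texture alpha is multiplied by particle alpha.
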Material /Game/Gadgets/Assets/BottleRocket/M_DebrisSmall_01_GPU
graph {
 properties {
  bUsedWithParticleSprites=true
  blendMode=BLEND_Translucent
  pixelDepthOffsetMode=PDOM_Legacy
  shadingModel=MSM_Unlit
  usedShadingModels=Unlit
 }
 nodes [
  n1 = ScalarParameter() {defaultValue=2, parameterName=Rows_Columns}
  n2 = TextureSample(UVs=MaterialFunctionCall!(UseTexCoord 1, Rows=@n1, Columns=@n1, Particle Random Value) {materialFunction=/Game/Effects/Fort_Effects/Materials/_Functions/MF_RandomSubUV_GPU_Sprites}.Result) {texture=&/T_DebrisSmall_01}
  n3 = ParticleColor()
  MaterialFunctionCall(Emissive=Multiply(A=@n2.RGB, B=@n3.RGB)) {materialFunction=&MF_FX_EmissiveAdjustments}
  Multiply(A=@n2.A, B=@n3.A)
 ]
}
